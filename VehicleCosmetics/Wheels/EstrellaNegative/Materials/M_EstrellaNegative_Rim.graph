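# Unlit wheel rim material for the EstrellaNegative vehicle cosmetic, with an animated glow.
# Mask channels combine via Max, with red modulated by panning noise and green by rotating noise.
# That mask blends Painted Color toward its optional inverse, or a fixed red when unpainted, then mask plus Fresnel scales the result.
# The same computation feeds base color and emissive, passed through shared vehicle damage and de-rez transition functions.
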Material /VehicleCosmetics/Wheels/EstrellaNegative/Materials/M_EstrellaNegative_Rim
graph {
 properties {
  bAutomaticallySetUsageInEditor=false
  bUseMaterialAttributes=true
  bUsedWithMeshDeformer=true
  bUsedWithSkeletalMesh=true
  previewMesh=/VehicleCosmetics/Wheels/Glitch/Meshes/SK_Wheel_Glitch
  shadingModel=MSM_Unlit
  usedShadingModels=Unlit
 }
 nodes [
  n1 = NamedRerouteUsage()
  n2 = NamedRerouteUsage()
  MaterialFunctionCall(In=MaterialFunctionCall(Material=SetMaterialAttributes!(MaterialAttributes, Base Color=@n1, Emissive Color=@n2) {attributeSetTypes [#1, #2]}) {materialFunction=&MF_BaseVehicle_Damage}.Result) {materialFunction=&MF_VehicleDeRez_Transition}
  n3 = VectorParameter() {group=00 Base, parameterName=Painted Color, sortPriority=0, defaultValue=<rgba 0.2918 0.0273 0.0222 1>}
  n4 = OneMinus(None=@n3.RGB)
  n5 = TextureSampleParameter2D() {
   group=00 Base
   parameterName=Masks
   samplerType=SAMPLERTYPE_LinearColor
   texture=/VehicleCosmetics/Wheels/EstrellaNegative/Textures/T_Estrella_Negative_Rim_RGB1
  }
  n6 = Multiply(A=@n5.R) {constB=7}
  n7 = TextureSample(UVs=Panner(Coordinate=TextureCoordinate() {vTiling=-2}) {speedY=0.25}) {samplerSource=SSM_Wrap_WorldGroupSettings, texture=/VehicleCosmetics/Wheels/_SharedTextures/FX/Universal_Noise_34_RGB}
  n8 = Multiply(A=@n5.G) {constB=5}
  n9 = Add(A=TextureCoordinate(), B=Constant2Vector() {r=0.125})
  n10 = Clamp(None=TextureSample(UVs=Rotator(Coordinate=@n9) {speed=-0.5}) {samplerSource=SSM_Wrap_WorldGroupSettings, texture=/VehicleCosmetics/Wheels/_SharedTextures/FX/Universal_Noise_51_RGB}.B)
  n11 = Max(A=Clamp(None=@n5.B), B=Max(A=Multiply(A=@n6, B=@n7.R) {constB=7}, B=Multiply(A=@n8, B=@n10) {constB=5}))
  n12 = StaticSwitchParameter(
   True=LinearInterpolate(A=@n3.RGB, B=StaticSwitchParameter(True=@n4, False=@n3.RGB) {group=00 Base, parameterName=INV, sortPriority=1}, Alpha=@n11)
   False=LinearInterpolate(A=@n3.RGB, B=Constant3Vector() {constant=<rgba 0.45 0.01 0 0>}, Alpha=@n11)
  ) {defaultValue=true, group=00 Base, parameterName=Painted, sortPriority=2}
  n13 = Add(
   A=@n11
   B=MaterialFunctionCall!(Normal, Exponent=Constant() {r=5}) {materialFunction=/VehicleCosmetics/Wheels/Cola_Holographic/Materials/MF_Fresnel_Classic}.Result
  )
  NamedRerouteDeclaration(None=Multiply(A=@n12, B=@n13)) {name=BC, nodeColor=<rgba 1 0 0.5859 1>}
  NamedRerouteDeclaration(None=Multiply(A=@n12, B=@n13)) {name=Emissive, nodeColor=<rgba 1 0 0.5859 1>}
 ]
}
