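# Translucent unlit material for the EstrellaNegative wheel's black hole effect, wrapped in vehicle damage and de-rez transition functions.
# Panning noise in UV and radial-texture coordinates, apparently blended by red vertex color, builds nebula, sparkle dots and a Fresnel-based mask.
# These drive emissive and opacity, with a pulsing center glow and optional hue-shifted paint tint.
# Scalable world position offset combines object-relative displacement with a sine wave.
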Material /VehicleCosmetics/Wheels/EstrellaNegative/Materials/M_Wheel_EstrellaNegative_BlackHole
graph {
 properties {
  bAutomaticallySetUsageInEditor=false
  bUseMaterialAttributes=true
  bUsedWithMeshDeformer=true
  bUsedWithSkeletalMesh=true
  blendMode=BLEND_Translucent
  previewMesh=/VehicleCosmetics/Wheels/Glitch/Meshes/SK_Wheel_Glitch
  shadingModel=MSM_Unlit
  usedShadingModels=Unlit
 }
 nodes [
  n1 = NamedRerouteUsage()
  n2 = NamedRerouteUsage()
  n3 = NamedRerouteUsage()
  MaterialFunctionCall(In=MaterialFunctionCall(Material=SetMaterialAttributes!(MaterialAttributes, Emissive Color=@n1, Opacity=@n2, World Position Offset=@n3) {attributeSetTypes [#1, #2, #3]}) {materialFunction=&MF_BaseVehicle_Damage}.Result) {materialFunction=&MF_VehicleDeRez_Transition}
  n4 = ObjectRadius()
  n5 = Subtract(A=WorldPosition().XYZ, B=ObjectPositionWS())
  n6 = TextureCoordinate()
  n7 = Sine(None=ComponentMask(None=Panner(Coordinate=@n6) {speedX=1, speedY=1}) {r=true})
  n8 = Add(
   A=Multiply(A=Multiply(A=@n4, B=@n5)) {constB=-2}
   B=Multiply(A=Multiply(A=@n7) {constB=300}, B=Constant3Vector() {constant=<rgba 1 1 1 0>}) {constB=300}
  )
  NamedRerouteDeclaration(None=Multiply(A=ScalarParameter() {group=00 Base, parameterName=WPO Multiplier, sortPriority=5}, B=Divide(A=@n8) {constB=1000})) {name=WPO, nodeColor=<rgba 1 0 0.6328 1>}
  NamedRerouteDeclaration(None=VertexColor().R) {name=VC_R, nodeColor=<rgba 1 0.5859 0 1>}
  n9 = NamedRerouteUsage()
  n10 = TextureCoordinate()
  n11 = Constant2Vector() {g=0.5, r=1}
  n12 = TextureSample(UVs=Panner(Coordinate=Multiply(A=@n10, B=@n11)) {speedX=0.1, speedY=0.1}) {samplerSource=SSM_Wrap_WorldGroupSettings, texture=/VehicleCosmetics/Wheels/_SharedTextures/FX/Universal_Noise_55_RGB}
  n13 = Multiply(
   A=TextureSample(UVs=Panner(Coordinate=@n10) {speedX=0.2, speedY=0.1}) {samplerSource=SSM_Wrap_WorldGroupSettings, texture=&Universal_Noise_55_RGB}.G
   B=@n12.B
  )
  n14 = TextureSample() {
   samplerSource=SSM_Wrap_WorldGroupSettings
   samplerType=SAMPLERTYPE_LinearColor
   texture=/VehicleCosmetics/Wheels/_SharedTextures/FX/Radial_Coords01_D
  }
  n15 = AppendVector(A=@n14.R, B=@n14.R)
  n16 = Multiply(A=@n15, B=Constant2Vector() {g=0.3, r=1})
  n17 = TextureSample() {samplerSource=SSM_Wrap_WorldGroupSettings, texture=/VehicleCosmetics/Wheels/_SharedTextures/FX/Radial_Generic_01_Pack}
  n18 = Clamp(None=@n17.B)
  n19 = Add(A=Time(), B=@n18)
  n20 = TextureSample(UVs=Panner(Coordinate=Add(A=@n16, B=@n18), Time=@n19) {speedX=0.1, speedY=-0.2}) {samplerSource=SSM_Wrap_WorldGroupSettings, texture=&Universal_Noise_55_RGB}
  n21 = LinearInterpolate(A=@n20.G, B=@n13, Alpha=@n9)
  n22 = Constant() {r=8}
  n23 = Constant() {r=16}
  n24 = OneMinus(None=LinearInterpolate(
    A=MaterialFunctionCall!(Normal, Exponent=@n22) {materialFunction=/VehicleCosmetics/Wheels/Cola_Holographic/Materials/MF_Fresnel_Classic}.Result
    B=MaterialFunctionCall(Exponent=@n23) {materialFunction=&MF_Fresnel_Classic}.Result
    Alpha=@n9
   ))
  n25 = Multiply(A=Add(A=Multiply(A=@n21) {constB=0.25}, B=Subtract(A=@n24) {constB=0.9})) {constB=10}
  NamedRerouteDeclaration(None=OneMinus(None=Clamp(None=@n25))) {name=Mask for Main Lerps, nodeColor=<rgba 0 1 0.0859 1>}
  NamedRerouteDeclaration(None=LinearInterpolate(A=@n20.G, B=@n13, Alpha=@n9)) {name=Nebula, nodeColor=<rgba 0 0.1562 1 1>}
  n26 = NamedRerouteUsage()
  n27 = LinearInterpolate(A=Constant3Vector() {constant=<rgba 0.04 0.05 0.08 0>}, B=Constant3Vector() {constant=<rgba 0.15 0.081 0.21 0>}, Alpha=@n26)
  n28 = VectorParameter() {group=00 Base, parameterName=Paint Color, sortPriority=0, defaultValue=<rgba 1 0.7011 0.7011 1>}
  n29 = Multiply(A=Multiply(A=Desaturation(None=@n27) {luminanceFactors=<rgba 0.3 0.59 0.11 0>}) {constB=2}, B=@n28.RGB) {constB=2}
  n30 = Constant() {r=0.075}
  n31 = StaticSwitchParameter(
   True=LinearInterpolate(
    A=@n29
    B=MaterialFunctionCall!(Color=@n29, Angle=@n30) {materialFunction=/VehicleCosmetics/Wheels/Concept_Inverted/Materials/MF_HUE_Shift}.Result
    Alpha=@n26
   )
   False=@n27
  ) {group=00 Base, parameterName=Painted, sortPriority=1}
  n32 = Multiply(A=NamedRerouteUsage()) {constB=5}
  n33 = Multiply(A=@n28.RGB) {constB=6}
  n34 = NamedRerouteUsage()
  NamedRerouteDeclaration(None=LinearInterpolate(A=LinearInterpolate(A=@n31, B=@n33, Alpha=@n32), B=Multiply(A=@n28.RGB) {constB=2}, Alpha=@n34)) {name=Emissive, nodeColor=<rgba 0.4766 1 0 1>}
  n35 = ConstantBiasScale(None=OneMinus(None=@n13)) {bias=-0.75, scale=2}
  n36 = Panner(Coordinate=Multiply(A=@n10, B=Constant2Vector() {g=0.75, r=1})) {speedX=0.5, speedY=0.25}
  n37 = Power(Base=Multiply(
    A=Clamp(None=@n35)
    B=TextureSample(UVs=@n36) {samplerSource=SSM_Wrap_WorldGroupSettings, texture=/VehicleCosmetics/Wheels/_SharedTextures/FX/skin_twankle2_rgb}.R
   ))
  n38 = Power(Base=@n12.R)
  n39 = Multiply(A=@n15, B=Constant2Vector() {g=0.5, r=3})
  n40 = TextureSample(UVs=Panner(Coordinate=Add(A=@n39, B=@n18), Time=@n19) {speedX=0.5, speedY=-0.2}) {samplerSource=SSM_Wrap_WorldGroupSettings, texture=&skin_twankle2_rgb}
  n41 = Max(A=Power(Base=@n20.R), B=Power(Base=Multiply(A=@n20.G, B=@n40.R)))
  n42 = TextureCoordinate()
  n43 = Multiply(
   A=TextureSample(UVs=Rotator(Coordinate=@n42) {speed=0.01}) {samplerSource=SSM_Wrap_WorldGroupSettings, texture=/VehicleCosmetics/Wheels/_SharedTextures/FX/Universal_Noise_51_RGB}.G
   B=TextureSample(UVs=Rotator(Coordinate=@n42) {speed=0.1}) {samplerSource=SSM_Wrap_WorldGroupSettings, texture=&Universal_Noise_51_RGB}.G
  )
  n44 = Time()
  n45 = Sine(None=Multiply(A=@n44) {constB=0.1})
  n46 = Add(A=@n17.R, B=Multiply(A=Multiply(A=@n17.G, B=@n43), B=Abs(None=@n45)))
  n47 = ScalarParameter() {defaultValue=4.25, group=00 Base, parameterName=Center Glow Power, sortPriority=2}
  n48 = Multiply(A=@n44) {constB=0.01}
  n49 = Multiply(
   A=Multiply(A=Power(Base=@n46, Exp=@n47), B=ScalarParameter() {defaultValue=1, group=00 Base, parameterName=Center Glow Multiply, sortPriority=3})
   B=Abs(None=Sine(None=@n48))
  )
  NamedRerouteDeclaration(None=LinearInterpolate(A=Add(A=@n41, B=@n49), B=Max(A=@n37, B=@n38), Alpha=@n9)) {name=Dots, nodeColor=<rgba 1 0 0.7969 1>}
  n50 = ScalarParameter() {defaultValue=2, group=00 Base, parameterName=Opacity, sortPriority=4}
  n51 = Clamp(None=@n32)
  NamedRerouteDeclaration(None=LinearInterpolate(A=LinearInterpolate(A=@n50, Alpha=@n51), Alpha=@n34)) {name=Opacity, nodeColor=<rgba 0.4766 1 0 1>}
 ]
}
